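; Blueprint that plays diegetic party music from its Audio component, placed among the NeonCity speaker cue presets.
; StartMusicParty plays and fades in the audio, and StopMusicParty fades it out, both only on the Remote (non-authority) branch of an authority switch.
; The stop event's FadeOutDuration input is unused in favor of a fixed fade time, and BeginPlay and Tick are empty.
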(event EventBeginPlay)

(event EventStartMusicParty
  (bind _audio (Variables|Default|GetAudio))
  (Networking|SwitchHasAuthority
    (:Remote
      (Audio|Components|Audio|Play _audio 0.0)
      (Audio|Components|Audio|FadeIn _audio 0.3 1.0 0.0 "Linear" :!class /Script/Engine.AudioComponent))))

(event EventTick
  (params (DeltaSeconds :type Float%20%28single-precision%29)))

(event EventStopMusicParty
  (params (FadeOutDuration :type Float%20%28double-precision%29))
  (Networking|SwitchHasAuthority
    (:Remote
      (Audio|Components|Audio|FadeOut (Variables|Default|GetAudio) 2.0 0.0 "Linear" :!class /Script/Engine.AudioComponent))))

(fn ConstructionScript)
